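; Red dot assault rifle weapon Blueprint that animates scope sway and recoil and hides muzzle and shell FX while aiming.
; Local firing drives a FireShake timeline and aimed-only camera shake, looped at fire rate when Quartz gunfire is enabled.
; An ADS timeline eases the scope in.
; Tick combines aim, recoil, look and strafe offsets through the native scope component, or client-side Blueprint interpolation while targeting as fallback.
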(event EventBeginPlay
  ;[1 1 1] Cache firing rate
  (|CallParentFunction)
  (bind _outrow (Fortnite|GetRangedWeaponStatsRow (Data|GetWeaponDataWeaponStatHandle)))
  (Variables|Default|SetFiring%20Rate (.FiringRate (Utilities|Struct|BreakFortRangedWeaponStats _outrow))))

(event EventOnPlayWeaponFireFX
  (params (bPersistentFire :type Boolean) (bSecondaryFire :type Boolean))
  (bind _returnvalue (Sounds|IsQuartzGunfireEnabled))
  (bind _fireshake (Variables|B_Assault_Chrono_RedDotAR_Athena_Weap|GetFireShake))
  ;[1 1 1] Recoil Scope Mesh Animation
  (|CallParentFunction bPersistentFire bSecondaryFire)
  (if (Pawn|IsLocallyControlled (Variables|Actor|GetInstigator))
    (Default|ToggleFireFX (not (FortWeapon|IsTargeting)))
    (Components|Timeline|SetTimelineLength _fireshake (/ 1.0 (Variables|Default|GetFiring%20Rate)))
    (Components|Timeline|SetLooping _fireshake _returnvalue)
    (CallFunction|RecoilCamShake)
    (Variables|Default|SetFiring true)
    (if _returnvalue
      (bind _returnvalue_2 (Utilities|Time|SetTimerbyEvent (EventDispatchers|CreateEvent :!event Custom|RecoilCamShake) (/ 1.0 (Variables|Default|GetFiring%20Rate)) true))
      (Variables|Default|SetRecoilTimerHandle _returnvalue_2)
      (else
        _))
    (bind (_direction _newtrack_0) (Timeline|FireShake PlayFromStart)
      (:Update
        (Variables|Scope|SetScope%20Camera%20Offset%20Amount (Math|Float|Lerp 29.0 28.0 _newtrack_0))
        (Variables|Default|SetXRecoilOffset (Math|Float|Lerp 0.0 (Math|Random|RandomFloatinRange -0.8 0.8) _newtrack_0))
        (Variables|Default|SetYRecoilOffset (Math|Float|Lerp 0.0 0.3 _newtrack_0)))
      (:Finished
        (if (not (Sounds|IsQuartzGunfireEnabled))
          (Variables|Default|SetFiring false))))))

(event EventOnStopWeaponFireFX
  ;[1 1 1] Recoil Scope Mesh Animation
  (|CallParentFunction)
  (if (Pawn|IsLocallyControlled (Variables|Actor|GetInstigator))
    (Utilities|Time|ClearandInvalidateTimerbyHandle (Variables|Default|GetRecoilTimerHandle))
    (Components|Timeline|SetLooping (Variables|B_Assault_Chrono_RedDotAR_Athena_Weap|GetFireShake) false)
    (Variables|Default|SetFiring false)))

(event Custom|RecoilCamShake
  ;[1 1 1] Recoil Scope Mesh Animation
  (if (Camera|GetIsTargeting)
    (CameraShakes|StartCameraShake :self (Variables|/Script/Engine.PlayerController|GetPlayerCameraManager (Game|GetPlayerController 0)) :ShakeClass "/FlipperGameplay/Items/Weapons/RedDotAR/CamShake_Athena_RedDot_AR_Recoil.CamShake_Athena_RedDot_AR_Recoil_C" :Scale 1.0 :PlaySpace "CameraLocal")))

(event EventOnSetTargeting
  (params (bNewIsTargeting :type Boolean))
  (bind _previousrotation (Variables|Default|GetPreviousRotation))
  (bind _yaw (.yaw _previousrotation))
  (bind _returnvalue (Game|GetPlayerController 0))
  (bind _returnvalue_2 (Pawn|GetControlRotation _returnvalue :!class /Script/Engine.Controller))
  (bind _yaw_2 (.yaw _returnvalue_2))
  (bind _returnvalue_3 (Game|GetPlayerController 0))
  (bind _returnvalue_4 (Pawn|GetControlRotation _returnvalue_3 :!class /Script/Engine.Controller))
  (bind _pitch (.pitch _returnvalue_4))
  (bind _previousrotation_2 (Variables|Default|GetPreviousRotation))
  (bind _pitch_2 (.pitch _previousrotation_2))
  (bind _returnvalue_5 (Pawn|GetControlledPawn _returnvalue))
  (bind _returnvalue_6 (Transformation|GetVelocity _returnvalue_5))
  (bind _returnvalue_7 (Math|Vector|UnrotateVector _returnvalue_6 _returnvalue_2))
  (bind _y (.y _returnvalue_7))
  (bind _returnvalue_8 (+ _yaw_2 3.6e+02))
  (bind _returnvalue_9 (+ _pitch 3.6e+02))
  (bind _returnvalue_10 (+ _pitch_2 3.6e+02))
  (bind _returnvalue_11 (+ _yaw 3.6e+02))
  (bind _firing (Variables|Default|GetFiring))
  (|CallParentFunction bNewIsTargeting)
  (if bNewIsTargeting
    (Utilities|FlowControl|Sequence
      (:then_0
        ;[1 1 1] ADS animate in
        (bind (_direction _newtrack_0) (Timeline|ADS PlayFromStart)
          (:Update
            (Variables|Default|SetADSAlpha _newtrack_0)
            (bind _output_get (Variables|Default|SetYADSOffset (Math|Float|Lerp (Variables|Default|GetPreADSYOffset) 0.0 _newtrack_0)))
            (bind _output_get_2 (Variables|Default|SetXADSOffset (Math|Float|Lerp (Variables|Default|GetPreADSXOffset) 0.0 _newtrack_0))))))
      (:then_1
        (if (Variables|Default|GetFiring)
          (Default|ToggleFireFX false))))
    (else
      (Utilities|FlowControl|Sequence
        (:then_0
          (if (Scope|UseNativeCalcScopeOffset (Variables|Default|GetScopeAnimComp))
            (Scope|ResetScopeRotation (Variables|Default|GetScopeAnimComp))
            (else
              (Utilities|DoOnceThenDo false
                (:DoOnce
                  (Variables|Default|SetPreviousRotation (Pawn|GetControlRotation (Game|GetPlayerController 0) :!class /Script/Engine.Controller)))
                (:ThenDo
                  (Utilities|FlowControl|Sequence
                    (:then_0
                      ;[1 1 1] Set Y Axis Interpolation Target Depending on whether current control rotation is different than the previous
                      (if (Math|Float|NearlyEqual(Float) _returnvalue_10 _returnvalue_9 0.1)
                        (Variables|Default|SetYOffsetInterpTarget 0.0)
                        (elif (> _returnvalue_10 _returnvalue_9)
                          (Variables|Default|SetYOffsetInterpTarget 0.3)
                          (else
                            (Variables|Default|SetYOffsetInterpTarget -0.3)))))
                    (:then_1
                      ;[1 1 1] Set X Axis Interpolation Target Depending on whether current control rotation is different than the previous
                      (Utilities|FlowControl|Sequence
                        (:then_0
                          (if (Math|Float|NearlyEqual(Float) _returnvalue_11 _returnvalue_8 0.1)
                            (Variables|Default|SetXOffsetInterpTarget 0.0)
                            (elif (> _returnvalue_11 _returnvalue_8)
                              (Variables|Default|SetXOffsetInterpTarget 0.5)
                              (else
                                (Variables|Default|SetXOffsetInterpTarget -0.5)))))
                        (:then_1
                          ;[1 1 1] OffSet X Axis Interpolation Target based on strafing velocity
                          (if (Math|Float|NearlyEqual(Float) _y 0.0 0.1)
                            (Variables|Default|SetXStrafeOffset 0.0)
                            (elif (< _y 0)
                              (Variables|Default|SetXStrafeOffset 0.25)
                              (else
                                (Variables|Default|SetXStrafeOffset -0.25)))))))
                    (:then_2
                      (Variables|Default|SetPreviousRotation (Pawn|GetControlRotation (Game|GetPlayerController 0) :!class /Script/Engine.Controller)))))))))
        (:then_1
          (if _firing
            (Default|ToggleFireFX _firing)))))))

(event EventTick
  (params (DeltaSeconds :type Float%20%28single-precision%29))
  (bind _previousrotation (Variables|Default|GetPreviousRotation))
  (bind _yaw (.yaw _previousrotation))
  (bind _returnvalue (Game|GetPlayerController 0))
  (bind _returnvalue_2 (Pawn|GetControlRotation _returnvalue :!class /Script/Engine.Controller))
  (bind _yaw_2 (.yaw _returnvalue_2))
  (bind _returnvalue_3 (Game|GetPlayerController 0))
  (bind _returnvalue_4 (Pawn|GetControlRotation _returnvalue_3 :!class /Script/Engine.Controller))
  (bind _pitch (.pitch _returnvalue_4))
  (bind _previousrotation_2 (Variables|Default|GetPreviousRotation))
  (bind _pitch_2 (.pitch _previousrotation_2))
  (bind _returnvalue_5 (Pawn|GetControlledPawn _returnvalue))
  (bind _returnvalue_6 (Transformation|GetVelocity _returnvalue_5))
  (bind _returnvalue_7 (Math|Vector|UnrotateVector _returnvalue_6 _returnvalue_2))
  (bind _y (.y _returnvalue_7))
  (bind _returnvalue_8 (+ _yaw_2 3.6e+02))
  (bind _returnvalue_9 (+ _pitch 3.6e+02))
  (bind _returnvalue_10 (+ _pitch_2 3.6e+02))
  (bind _returnvalue_11 (+ _yaw 3.6e+02))
  (if (Scope|UseNativeCalcScopeOffset (Variables|Default|GetScopeAnimComp))
    (Scope|CalcScopeOffset (Variables|Default|GetScopeAnimComp) DeltaSeconds (Variables|Default|GetXRecoilOffset) (Variables|Default|GetYRecoilOffset) (Variables|Default|GetXADSOffset) (Variables|Default|GetYADSOffset) (Variables|Scope|GetScopeCameraOffsetX) (Variables|Scope|GetScopeCameraOffsetY))
    (elif (Networking|IsServer)
      _
      (else
        (bind _returnvalue_12 (GameplayTags|HasMatchingGameplayTag(Message) (Pawn|GetControlledPawn (Game|GetPlayerController 0)) "(TagName=\"Gameplay.Status.IsTargeting\")"))
        (if _returnvalue_12
          (Utilities|FlowControl|Sequence
            (:then_0
              (Utilities|DoOnceThenDo false
                (:DoOnce
                  (Variables|Default|SetPreviousRotation (Pawn|GetControlRotation (Game|GetPlayerController 0) :!class /Script/Engine.Controller)))
                (:ThenDo
                  (Utilities|FlowControl|Sequence
                    (:then_0
                      ;[1 1 1] Set Y Axis Interpolation Target Depending on whether current control rotation is different than the previous
                      (if (Math|Float|NearlyEqual(Float) _returnvalue_10 _returnvalue_9 0.1)
                        (Variables|Default|SetYOffsetInterpTarget 0.0)
                        (elif (> _returnvalue_10 _returnvalue_9)
                          (Variables|Default|SetYOffsetInterpTarget 0.3)
                          (else
                            (Variables|Default|SetYOffsetInterpTarget -0.3)))))
                    (:then_1
                      ;[1 1 1] Set X Axis Interpolation Target Depending on whether current control rotation is different than the previous
                      (Utilities|FlowControl|Sequence
                        (:then_0
                          (if (Math|Float|NearlyEqual(Float) _returnvalue_11 _returnvalue_8 0.1)
                            (Variables|Default|SetXOffsetInterpTarget 0.0)
                            (elif (> _returnvalue_11 _returnvalue_8)
                              (Variables|Default|SetXOffsetInterpTarget 0.5)
                              (else
                                (Variables|Default|SetXOffsetInterpTarget -0.5)))))
                        (:then_1
                          ;[1 1 1] OffSet X Axis Interpolation Target based on strafing velocity
                          (if (Math|Float|NearlyEqual(Float) _y 0.0 0.1)
                            (Variables|Default|SetXStrafeOffset 0.0)
                            (elif (< _y 0)
                              (Variables|Default|SetXStrafeOffset 0.25)
                              (else
                                (Variables|Default|SetXStrafeOffset -0.25)))))))
                    (:then_2
                      (Variables|Default|SetPreviousRotation (Pawn|GetControlRotation (Game|GetPlayerController 0) :!class /Script/Engine.Controller)))))))
            (:then_1
              ;[1 1 1] Tick to interpolate the Scope Mesh
              (bind _output_get (Variables|Default|SetYOffset (Math|Interpolation|FInterpTo (Variables|Default|GetYOffset) (Variables|Default|GetYOffsetInterpTarget) DeltaSeconds 5.0)))
              (Variables|Scope|SetScopeCameraOffsetY (Utilities|Operators|Add _output_get (Variables|Default|GetYADSOffset) (Variables|Default|GetYRecoilOffset)))
              (bind _output_get_2 (Variables|Default|SetXOffset (Math|Interpolation|FInterpTo (Variables|Default|GetXOffset) (Utilities|Operators|Add (Variables|Default|GetXOffsetInterpTarget) (Variables|Default|GetXStrafeOffset) (Variables|Default|GetXRecoilOffset)) DeltaSeconds 5.0)))
              (Variables|Scope|SetScopeCameraOffsetX (Utilities|Operators|Add _output_get_2 (Variables|Default|GetXADSOffset) 0)))))))))

(fn ToggleFireFX
  (params (Enable :type Boolean))
  (if Enable
    (CallFunction|ActivateShellsFX true)
    (CallFunction|MuzzleFlashFX false)
    (else
      (CallFunction|DeactivateMuzzleFX)
      (CallFunction|DeactivateAllShellsFX))))

(fn ConstructionScript
  (|CallParentFunction))
